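; Passive gameplay ability that credits a storm phase survival quest as the safe zone advances.
; On activation it listens through the game state's event router for the stateful safe zone phase update event and passes the phase index to StormZoneUpdated.
; For phases above 1, that event sends a custom stat event with the quest source tag to the activating player's quest manager.
; Ending the ability stops the stored listener task.
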(event Ability|EventActivateAbility
  (|CallParentFunction)
  (bind _returnvalue (GameplayEventRouter|GetEventRouter "/Script/FortniteGame.FortGameState" (Ability|GetAvatarActorfromActorInfo)))
  (bind (_asynctaskproxy _payload) (GameplayEventRouter|StartListeningToStatefulEvent _returnvalue "/Script/FortniteGame.FortSafeZonePhaseUpdatedEvent")
    (:then
      (Variables|Default|SetSafeZonePhaseUpdatedAsyncTask _asynctaskproxy))
    (:OnEventReceived
      _)
    (:OnSavedEventStateExists
      _))
  (bind _payload_safezonephaseindex (Utilities|Struct|BreakFortSafeZonePhaseUpdatedEvent _payload))
  (CallFunction|StormZoneUpdated _payload_safezonephaseindex))

(event Ability|EventOnEndAbility
  (params (bWasCancelled :type Boolean))
  (bind _safezonephaseupdatedasynctask (Variables|Default|GetSafeZonePhaseUpdatedAsyncTask)
    (:then
      (GameplayEventRouter|StopListeningtoStatefulEvent _safezonephaseupdatedasynctask))
    (:else)))

(event Custom|Storm%20Zone%20Updated
  (params (SafeZonePhase :type Integer))
  (if (> SafeZonePhase 1)
    (Quest|SendComplexCustomStatEvent (Quests|GetQuestManager (Fortnite|GetFortPlayerControllerfromActor (Ability|GetActivatingPawn)) "Invalid") 0 (Variables|Default|GetQuestSourceTag) "(GameplayTags=)" 1)))

(event Ability|EventFailedToActivatePassiveAbility
  (params (ActorInfo :type Gameplay%20Ability%20Actor%20Info%20Structure))
  (|CallParentFunction))
